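# Sound cue for the drop sound of the BigBushGrenade item.
# A Random node picks with equal weight between two LeafGrenade drop waves at reduced volume.
# It uses the shared weapon drop attenuation and the gadget items sound class.
# A concurrency override caps simultaneous instances at a small count and prevents new ones beyond that.
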SoundCue /BigBushGrenade/Items/Audio/BigBushGrenade_Drop_Cue
root=Random_0
nodes [
 {id=Random_0, children [WavePlayer_2,WavePlayer_3], properties {weights [1,1]}}
 {id=WavePlayer_2, properties {soundWaveAssetPtr=&/LeafGrenade_Drop_02}}
 {id=WavePlayer_3, properties {soundWaveAssetPtr=&/LeafGrenade_Drop_01}}
]
properties {
 attenuationSettings=&Fort_Weapon_Drop_Att
 bOverrideConcurrency=true
 maxDistance=3000
 soundClassObject=&Fort_Gadgets_Items
 volumeMultiplier=0.5
 concurrencyOverrides {maxCount=3, resolutionRule=PreventNew}
}
