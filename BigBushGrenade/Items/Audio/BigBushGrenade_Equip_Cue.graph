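# Sound cue for the Big Bush Grenade equip foley.
# A random node picks with equal weight between two Leaf Grenade equip wave variants.
# It plays on the weapon foley sound class with the shared weapon reload attenuation at roughly half volume, with concurrency override and debug enabled.
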SoundCue /BigBushGrenade/Items/Audio/BigBushGrenade_Equip_Cue
root=Random_0
nodes [
 {id=Random_0, children [WavePlayer_1,WavePlayer_0], properties {weights [1,1]}}
 {id=WavePlayer_0, properties {soundWaveAssetPtr=&/LeafGrenade_Equip_01}}
 {id=WavePlayer_1, properties {soundWaveAssetPtr=&/LeafGrenade_Equip_02}}
]
properties {
 attenuationSettings=&Fort_Weapon_Reload_Att
 bDebug=true
 bOverrideConcurrency=true
 maxDistance=2250
 soundClassObject=&Fort_WeaponFoley
 volumeMultiplier=0.501
}
