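# Sound cue for the BigBushGrenade exploding in water, layering a leaf grenade explosion with a water impact.
# A mixer blends one of three equally weighted leaf explosion waves at full volume with one of three grenade water impact waves at half volume.
# Only the water layer passes through a modulator.
# The cue uses far explosion attenuation, the large explosives sound class and a boosted overall volume.
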SoundCue /BigBushGrenade/Items/Audio/BigBushGrenade_Explode_Water_Cue
root=Mixer_0
nodes [
 {id=Mixer_0, children [Random_0,Modulator_1], properties {inputVolume [1,0.5]}}
 {id=Modulator_1, children [Random_1]}
 {id=Random_0, children [WavePlayer_2,WavePlayer_3,WavePlayer_6], properties {weights [1,1,1]}}
 {id=Random_1, children [WavePlayer_1,WavePlayer_4,WavePlayer_5], properties {weights [1,1,1]}}
 {id=WavePlayer_1, properties {soundWaveAssetPtr=/Game/Sounds/Fort_GamePlay_Sounds/Explosives/Grenade_Impact_Water_01}}
 {id=WavePlayer_2, properties {soundWaveAssetPtr=&/LeafGrenade_Explode_01}}
 {id=WavePlayer_3, properties {soundWaveAssetPtr=&/LeafGrenade_Explode_02}}
 {id=WavePlayer_4, properties {soundWaveAssetPtr=/Game/Sounds/Fort_GamePlay_Sounds/Explosives/Grenade_Impact_Water_02}}
 {id=WavePlayer_5, properties {soundWaveAssetPtr=/Game/Sounds/Fort_GamePlay_Sounds/Explosives/Grenade_Impact_Water_03}}
 {id=WavePlayer_6, properties {soundWaveAssetPtr=&/LeafGrenade_Explode_03}}
]
properties {
 attenuationSettings=/Game/Sounds/Attenuation/Fort_Explosion_Far_StW_Att
 maxDistance=11000
 soundClassObject=/Game/Sounds/FortSoundClassesAndModes/Fort_Explosives_Large_PSM
 volumeMultiplier=1.2383
}
