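# Sound cue for the looping in-air travel sound of the Big Bush Grenade projectile.
# A mixer layers a whistle loop with a louder, pitched-up foliage loop.
# The foliage layer has its own attenuation node with low-pass filtering and listener focus, while the cue carries separate in-air projectile attenuation settings.
# The mix runs through a third-person ranged-fire sound class, and a concurrency override prevents new instances at the limit.
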SoundCue /BigBushGrenade/Items/Audio/BigBushGrenade_InAirLoop_Cue
root=SoundClass_2
nodes [
 {
  id=Attenuation_0
  children [Modulator_0]
  properties {
   attenuationSettings=&Fort_Weapon_Projectile_Medium_Mono_Athena_Att
   bOverrideAttenuation=true
   attenuationOverrides {
    bAttenuateWithLPF=true
    bEnableListenerFocus=true
    distanceAlgorithm=Logarithmic
    falloffDistance=7000
    focusAzimuth=15
    focusReleaseInterpSpeed=2
    lPFFrequencyAtMin=800
    nonFocusVolumeAttenuation=0.7
    stereoSpread=800
    attenuationShapeExtents {x=1000}
   }
  }
 }
 {id=Mixer_0, children [WavePlayer_0, Attenuation_0], properties {inputVolume [1,3]}}
 {id=Modulator_0, children [WavePlayer_2], properties {pitchMax=1.5, pitchMin=1.5, volumeMax=1, volumeMin=1}}
 {id=SoundClass_2, children [Mixer_0], properties {soundClassOverride=&Fort_WeaponRanged_Fire_3P_PSM}}
 {id=WavePlayer_0, properties {bLooping=true, soundWaveAssetPtr=&/LeafGrenade_Travel_Whistle_Loop}}
 {id=WavePlayer_2, properties {bLooping=true, soundWaveAssetPtr=&/LeafGrenade_Travel_Foliage_Loop}}
]
properties {
 attenuationSettings=/Game/Sounds/Attenuation/Fort_Weapon_Projectile_InAir_Loop_Att
 bOverrideConcurrency=true
 maxDistance=8000
 soundClassObject=&Fort_WeaponRanged_Fire_3P_PSM
 volumeMultiplier=2
 attenuationOverrides {falloffDistance=5000, reverbWetLevelMax=0.7, reverbWetLevelMin=0.2, attenuationShapeExtents {x=1000}}
 concurrencyOverrides {resolutionRule=PreventNew}
}
