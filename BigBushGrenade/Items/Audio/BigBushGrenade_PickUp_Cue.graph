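# Sound cue for picking up the Big Bush Grenade item, built from leaf grenade pickup waves.
# A mixer blends a random, equally weighted choice of three pickup variations with a second layer that starts after a short fixed delay.
# It uses weapon pickup attenuation and the gadget items sound class, with a concurrency override that prevents new instances once the limit is reached.
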SoundCue /BigBushGrenade/Items/Audio/BigBushGrenade_PickUp_Cue
root=Mixer_0
nodes [
 {id=Delay_0, children [WavePlayer_4], properties {delayMax=0.1, delayMin=0.1}}
 {id=Mixer_0, children [Random_0,Delay_0], properties {inputVolume [0.75,0.7]}}
 {id=Random_0, children [WavePlayer_1,WavePlayer_2,WavePlayer_3], properties {weights [1,1,1]}}
 {id=WavePlayer_1, properties {soundWaveAssetPtr=&/LeafGrenade_Pickup_01}}
 {id=WavePlayer_2, properties {soundWaveAssetPtr=&/LeafGrenade_Pickup_02}}
 {id=WavePlayer_3, properties {soundWaveAssetPtr=&/LeafGrenade_Pickup_03}}
 {id=WavePlayer_4, properties {soundWaveAssetPtr=&/LeafGrenade_Pickup_Layer2}}
]
properties {
 attenuationSettings=&Fort_Weapon_Pickup_Att
 bOverrideConcurrency=true
 maxDistance=3500
 soundClassObject=&Fort_Gadgets_Items
 volumeMultiplier=0.4
 concurrencyOverrides {resolutionRule=PreventNew}
}
