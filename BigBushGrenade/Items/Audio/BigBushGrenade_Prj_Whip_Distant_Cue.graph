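# Sound cue for the distant flyby whip of the Big Bush Grenade projectile.
# It randomly picks one of three equally weighted leaf grenade distant flyby waves and passes it through a modulator.
# Playback uses the far grenade whip attenuation asset and the ranged weapon projectile effects sound class, with a boosted volume.
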SoundCue /BigBushGrenade/Items/Audio/BigBushGrenade_Prj_Whip_Distant_Cue
root=Modulator_0
nodes [
 {id=Modulator_0, children [Random_1]}
 {id=Random_1, children [WavePlayer_3,WavePlayer_4,WavePlayer_5], properties {weights [1,1,1]}}
 {id=WavePlayer_3, properties {soundWaveAssetPtr=&/LeafGrenade_Flyby_Distant_01}}
 {id=WavePlayer_4, properties {soundWaveAssetPtr=&/LeafGrenade_Flyby_Distant_02}}
 {id=WavePlayer_5, properties {soundWaveAssetPtr=&/LeafGrenade_Flyby_Distant_03}}
]
properties {
 attenuationSettings=/Game/Sounds/Attenuation/Fort_Grenade_Prj_Whip_Far_Att
 maxDistance=9000
 soundClassObject=&Fort_WeaponRanged_Projectile_Effects
 volumeMultiplier=1.5
 attenuationOverrides {
  bEnableReverbSend=false
  bUseComplexCollisionForOcclusion=true
  falloffDistance=1000
  occlusionVolumeAttenuation=0
  stereoSpread=0
  attenuationShapeExtents {x=512}
 }
}
