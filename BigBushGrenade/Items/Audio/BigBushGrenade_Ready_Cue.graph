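# Sound cue for the Big Bush Grenade's ready sound, played as a gadget item effect.
# A Random node picks one of two equally weighted LeafGrenade "Cook" waves and feeds a Modulator left at default settings for slight variation.
# It routes through the gadget items sound class, uses the shared gadget-placement attenuation asset, and plays at slightly reduced volume.
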SoundCue /BigBushGrenade/Items/Audio/BigBushGrenade_Ready_Cue
root=Modulator_0
nodes [
 {id=Modulator_0, children [Random_0]}
 {id=Random_0, children [WavePlayer_0,WavePlayer_1], properties {weights [1,1]}}
 {id=WavePlayer_0, properties {soundWaveAssetPtr=&/LeafGrenade_Cook_01}}
 {id=WavePlayer_1, properties {soundWaveAssetPtr=&/LeafGrenade_Cook_02}}
]
properties {
 attenuationSettings=/Game/Sounds/Attenuation/Fort_GadgetPlace_Att
 maxDistance=2000
 soundClassObject=&Fort_Gadgets_Items
 volumeMultiplier=0.8
}
