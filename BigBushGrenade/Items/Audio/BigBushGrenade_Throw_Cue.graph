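# Sound cue for the throw sound of the Big Bush Grenade item, categorized under the gadget items sound class.
# A random node picks one of three equally weighted leaf grenade throw waves on each play.
# It uses the shared gadget placement attenuation and overrides concurrency so new instances are blocked once the limit is reached, rather than cutting off ones already playing.
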SoundCue /BigBushGrenade/Items/Audio/BigBushGrenade_Throw_Cue
root=Random_0
nodes [
 {id=Random_0, children [WavePlayer_0,WavePlayer_1,WavePlayer_2], properties {weights [1,1,1]}}
 {id=WavePlayer_0, properties {soundWaveAssetPtr=&/LeafGrenade_Throw_01}}
 {id=WavePlayer_1, properties {soundWaveAssetPtr=&/LeafGrenade_Throw_02}}
 {id=WavePlayer_2, properties {soundWaveAssetPtr=&/LeafGrenade_Throw_03}}
]
properties {
 attenuationSettings=/Game/Sounds/Attenuation/Fort_GadgetPlace_Att
 bOverrideConcurrency=true
 maxDistance=2000
 soundClassObject=&Fort_Gadgets_Items
 volumeMultiplier=1
 concurrencyOverrides {resolutionRule=PreventNew}
}
