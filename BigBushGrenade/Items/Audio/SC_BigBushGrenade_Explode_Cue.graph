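# Explosion sound cue for the Big Bush Grenade, routed to the explosives sound class.
# A mixer blends a close layer and a quieter distant layer, each randomly picking one of three leaf grenade explosion waves, under a root modulator.
# Each layer has its own occlusion-enabled attenuation override.
# The distant layer adds distance low-pass filtering and a custom crossfade curve that fades it in away from the source, then out.
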SoundCue /BigBushGrenade/Items/Audio/SC_BigBushGrenade_Explode_Cue
root=Modulator_1
nodes [
 {
  id=Attenuation_1
  children [Random_0]
  properties {
   attenuationSettings=/Game/Sounds/Attenuation/Fort_Explosion_Far_Grenade_Att
   bOverrideAttenuation=true
   attenuationOverrides {
    bEnableFocusInterpolation=true
    bEnableOcclusion=true
    dBAttenuationAtMax=-40
    falloffDistance=2500
    occlusionLowPassFilterFrequency=8000
    occlusionTraceChannel=ECC_GameTraceChannel11
    occlusionVolumeAttenuation=1.5
    stereoSpread=1000
    attenuationShapeExtents {x=500}
   }
  }
 }
 {
  id=Attenuation_120
  children [Random_1]
  properties {
   attenuationSettings=&Fort_Weapon_Projectile_Distant_Mono_Athena_Att
   bOverrideAttenuation=true
   attenuationOverrides {
    bAttenuateWithLPF=true
    bEnableLogFrequencyScaling=true
    bEnableOcclusion=true
    distanceAlgorithm=Custom
    falloffDistance=10000
    hPFFrequencyAtMax=60
    lPFFrequencyAtMax=300
    lPFRadiusMax=30000
    lPFRadiusMin=12000
    occlusionLowPassFilterFrequency=5000
    occlusionTraceChannel=ECC_GameTraceChannel11
    occlusionVolumeAttenuation=0.8
    stereoSpread=1000
    attenuationShapeExtents {x=0}
    customAttenuationCurve {
     externalCurve=/Game/Sounds/Attenuation/Curves/Fort_WeaponFire_Crossfade_Distant_Athena
     editorCurveData {keys [
       <at 0 0>
       {arriveTangent=8.2315, interpMode=RCIM_Cubic, leaveTangent=8.2315, tangentMode=RCTM_User, time=0.0388, value=-0.0023}
       <at 0.2 1>
       {arriveTangent=0.0024, interpMode=RCIM_Cubic, leaveTangent=0.0024, time=0.35, value=1}
       {arriveTangent=-0.2026, interpMode=RCIM_Cubic, leaveTangent=-0.2026, tangentMode=RCTM_User, time=1, value=0}
      ]}
    }
   }
  }
 }
 {id=Mixer_0, children [Attenuation_1, Attenuation_120], properties {inputVolume [1,0.65]}}
 {id=Modulator_1, children [Mixer_0]}
 {id=Random_0, children [WavePlayer_0,WavePlayer_1,WavePlayer_2], properties {weights [1,1,1]}}
 {id=Random_1, children [WavePlayer_3,WavePlayer_4,WavePlayer_5], properties {weights [1,1,1]}}
 {id=WavePlayer_0, properties {soundWaveAssetPtr=&/LeafGrenade_Explode_01_v2}}
 {id=WavePlayer_1, properties {soundWaveAssetPtr=&/LeafGrenade_Explode_02_v2}}
 {id=WavePlayer_2, properties {soundWaveAssetPtr=&/LeafGrenade_Explode_03_v2}}
 {id=WavePlayer_3, properties {soundWaveAssetPtr=&/LeafGrenade_Explode_Distant_01}}
 {id=WavePlayer_4, properties {soundWaveAssetPtr=&/LeafGrenade_Explode_Distant_02}}
 {id=WavePlayer_5, properties {soundWaveAssetPtr=&/LeafGrenade_Explode_Distant_03}}
]
properties {maxDistance=10000, soundClassObject=/Game/Sounds/FortSoundClassesAndModes/Fort_Explosives, volumeMultiplier=1}
