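# Sound cue for the leaf (bush) grenade's drop impact, named as a UI drop-weapon impact sound.
# A random node picks with equal weight between two landing wave variations.
# It plays at half volume in the loot-collect sound class, with overridden attenuation for a short falloff range.
# A concurrency override caps simultaneous plays at three and blocks new ones beyond that.
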SoundCue /BigBushGrenade/Items/Audio/UI_DropWeaponImpact_LeafGrenade_Cue
root=Random_0
nodes [
 {id=Random_0, children [WavePlayer_0,WavePlayer_1], properties {weights [1,1]}}
 {id=WavePlayer_0, properties {soundWaveAssetPtr=&/LeafGrenade_Land_2}}
 {id=WavePlayer_1, properties {soundWaveAssetPtr=&/LeafGrenade_Land_3}}
]
properties {
 bOverrideAttenuation=true
 bOverrideConcurrency=true
 maxDistance=1000
 soundClassObject=&Fort_Loot_Collect
 volumeMultiplier=0.5
 attenuationOverrides {falloffDistance=1000, attenuationShapeExtents {x=0}}
 concurrencyOverrides {maxCount=3, resolutionRule=PreventNew}
}
