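; Shrub Bomb grenade projectile that puts a bush on players caught in its blast.
; Bounces and capsule overlaps end the fuse, triggering a one-time explosion; water entry or detonation traces for water and spawns a splash.
; On authority, the explosion attaches a bush actor to each hit player's pelvis, applies gameplay effects and runs a quest update reusing the DanceGrenade verb.
; On stop, it optionally spawns a building actor, gated by a scalable float.
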(event EventBeginPlay
  (|CallParentFunction))

(event OnComponentBeginOverlap(CapsuleComponent)
  (params (OverlappedComponent :type Primitive%20Component%20Object%20Reference) (OtherActor :type Actor%20Object%20Reference) (OtherComp :type Primitive%20Component%20Object%20Reference) (OtherBodyIndex :type Integer) (bFromSweep :type Boolean) (SweepResult :type Hit%20Result%20Structure%20%28by%20ref%29))
  (bind (_bblockinghit _binitialoverlap _time _distance _location _impactpoint _normal _impactnormal _physmat _hitactor) (Collision|BreakHitResult SweepResult))
  ;[1 1 1] If we overlap a pawn, fake a bounce.
  (bind _aspawn (Utilities|Casting|CastToPawn _hitactor))
  (if (Utilities|NotEqual(Object) _aspawn (Game|GetInstigator))
    (CallFunction|FuseEnded)))

(event EventOnBounce
  (params (Hit :type Hit%20Result%20Structure%20%28by%20ref%29))
  (bind (_bblockinghit _binitialoverlap _time _distance _location _impactpoint _normal _impactnormal _physmat _hitactor) (Collision|BreakHitResult Hit))
  ;[1 1 1] OnBounce - Start fuse
  (|CallParentFunction Hit)
  (bind _aspawn (Utilities|Casting|CastToPawn _hitactor)
    (:then
      (if (Utilities|NotEqual(Object) _aspawn (Game|GetInstigator))
        _))
    (:CastFailed
      _))
  (CallFunction|FuseEnded))

(event OnEnterWaterEvent(WaterInteractionComponent)
  (params (WaterBody :type Fort%20Water%20Body%20Actor%20Object%20Reference) (WaterInteractionComponent :type Fort%20Water%20Interaction%20Component%20Object%20Reference) (bIsFirstBody :type Boolean))
  (bind _mesh (Variables|B_Prj_Grenade|GetMesh))
  ;[1 1 1] Check to see if deploying in water and then Get water info
  (bind (_outhit _returnvalue) (Collision|LineTraceForObjects (+ (Transformation|GetActorLocation) "0, 0,100.000000") (- (Transformation|GetActorLocation) "0, 0,100.000000") (Utilities|Array|MakeArray "ObjectTypeQuery12")))
  (if _returnvalue
    (bind _asfort_water_body_actor (Utilities|Casting|CastToFortWaterBodyActor (.HitActor (Collision|BreakHitResult _outhit))))
    (Variables|Default|SetWaterLocation (.Location (Collision|BreakHitResult _outhit)))
    (bind _outwatersurfaceinfo (WaterBody|GetWaterSurfaceInfoStruct _asfort_water_body_actor (.Location (Collision|BreakHitResult _outhit)) true true false))
    (Rendering|Material|SetScalarParameterValueonMaterials _mesh "WaterDepth" (.WaterDepth (Utilities|Struct|BreakWaterSurfaceInfo _outwatersurfaceinfo)))
    (Rendering|Material|SetScalarParameterValueonMaterials _mesh "WaterIndex" (Math|Conversions|ToFloat(Integer) (.WaterBodyIdx (Utilities|Struct|BreakWaterSurfaceInfo _outwatersurfaceinfo))))
    (bind _returnvalue_2 (Effects|Components|ParticleSystem|SpawnFXAtLocation "/Game/Effects/Niagara/Systems/Enviroment/Water/NS_WaterSplash_Medium.NS_WaterSplash_Medium" (.WaterSurfaceLocation (Utilities|Struct|BreakWaterSurfaceInfo _outwatersurfaceinfo))))
    (Effects|Components|ParticleSystem|SetFloatParameter _returnvalue_2 "User.EffectScale" 1.75)
    (Effects|Components|ParticleSystem|SetFloatParameter _returnvalue_2 "User.WaterDepth" (.WaterDepth (Utilities|Struct|BreakWaterSurfaceInfo _outwatersurfaceinfo)))
    (Effects|Components|ParticleSystem|SetVectorParameter _returnvalue_2 "User.WaterPlaneLocation" (Utilities|Struct|BreakWaterSurfaceInfo _outwatersurfaceinfo))
    (Effects|Components|ParticleSystem|SetVectorParameter _returnvalue_2 "User.WaterPlaneNormal" (.WaterPlaneNormal (Utilities|Struct|BreakWaterSurfaceInfo _outwatersurfaceinfo)))))

(event EventFuseEnded
  (bind _mesh (Variables|B_Prj_Grenade|GetMesh))
  ;[1 1 1] Fuse ended - explode
  (Utilities|FlowControl|DoOnce false
    (:Completed
      (Delivery|DoExplosion)
      ;[1 1 1] Check to see if deploying in water and then Get water info
      (bind (_outhit _returnvalue) (Collision|LineTraceForObjects (+ (Transformation|GetActorLocation) "0, 0,100.000000") (- (Transformation|GetActorLocation) "0, 0,100.000000") (Utilities|Array|MakeArray "ObjectTypeQuery12")))
      (if _returnvalue
        (bind _asfort_water_body_actor (Utilities|Casting|CastToFortWaterBodyActor (.HitActor (Collision|BreakHitResult _outhit))))
        (Variables|Default|SetWaterLocation (.Location (Collision|BreakHitResult _outhit)))
        (bind _outwatersurfaceinfo (WaterBody|GetWaterSurfaceInfoStruct _asfort_water_body_actor (.Location (Collision|BreakHitResult _outhit)) true true false))
        (Rendering|Material|SetScalarParameterValueonMaterials _mesh "WaterDepth" (.WaterDepth (Utilities|Struct|BreakWaterSurfaceInfo _outwatersurfaceinfo)))
        (Rendering|Material|SetScalarParameterValueonMaterials _mesh "WaterIndex" (Math|Conversions|ToFloat(Integer) (.WaterBodyIdx (Utilities|Struct|BreakWaterSurfaceInfo _outwatersurfaceinfo))))
        (bind _returnvalue_2 (Effects|Components|ParticleSystem|SpawnFXAtLocation "/Game/Effects/Niagara/Systems/Enviroment/Water/NS_WaterSplash_Medium.NS_WaterSplash_Medium" (.WaterSurfaceLocation (Utilities|Struct|BreakWaterSurfaceInfo _outwatersurfaceinfo))))
        (Effects|Components|ParticleSystem|SetFloatParameter _returnvalue_2 "User.EffectScale" 1.75)
        (Effects|Components|ParticleSystem|SetFloatParameter _returnvalue_2 "User.WaterDepth" (.WaterDepth (Utilities|Struct|BreakWaterSurfaceInfo _outwatersurfaceinfo)))
        (Effects|Components|ParticleSystem|SetVectorParameter _returnvalue_2 "User.WaterPlaneLocation" (Utilities|Struct|BreakWaterSurfaceInfo _outwatersurfaceinfo))
        (Effects|Components|ParticleSystem|SetVectorParameter _returnvalue_2 "User.WaterPlaneNormal" (.WaterPlaneNormal (Utilities|Struct|BreakWaterSurfaceInfo _outwatersurfaceinfo)))))))

(event EventOnExploded
  (params (HitActors :type Array%20of%20Actor%20Object%20References) (HitResults :type Array%20of%20Hit%20Result%20Structures))
  ;[1 1 1] OnExploded - Emitter, Sound
  (|CallParentFunction HitActors HitResults)
  (Networking|SwitchHasAuthority
    (:Authority
      (for _array_element HitActors
        (bind _asfort_pawn (Utilities|Casting|CastToFortPawn _array_element))
        (bind _output_get (Variables|Default|SetAs%20Fort%20Pawn _asfort_pawn))
        (Utilities|FlowControl|Sequence
          (:then_0
            (bind _asplayer_pawn_athena (Utilities|Casting|CastToPlayerPawn_Athena _output_get))
            (bind _returnvalue (Game|SpawnActorfromClass "/BigBushGrenade/Items/Athena_Player_BushBomb.Athena_Player_BushBomb_C" (Math|Transform|MakeTransform (Math|Vector|MakeVector (.x (Transformation|GetActorLocation _asplayer_pawn_athena)) (.y (Transformation|GetActorLocation _asplayer_pawn_athena)) (- (.z (Transformation|GetActorLocation _asplayer_pawn_athena)) 1e+01)) (Transformation|GetActorRotation _asplayer_pawn_athena) "1.000000,1.000000,1.000000") "Undefined" "OverrideRootScale" _asplayer_pawn_athena 0 0 _asplayer_pawn_athena))
            (bind _output_get_2 (Variables|Default|SetPlayerBush _returnvalue))
            (Transformation|AttachActorToComponent _output_get_2 (Variables|/Script/Engine.Character|GetMesh _asplayer_pawn_athena) "pelvis" "SnapToTarget" "SnapToTarget" "KeepWorld" true) ; We have to attach to player mesh or the bush won't move properly with player.
          )
          (:then_1
            (Fortnite|ApplyGameplayEffectsToActor _output_get_2 (Variables|Default|GetEffects%20to%20Apply) (Variables|Default|GetSource%20Pos)))))
      ;[1 1 1] Quest update for hitting an enemy player and reset the count tracking the number of players hit
      (CallFunction|QuestUpdate)
      (Variables|Default|SetCount 0))))

(event EventOnStop
  (params (Hit :type Hit%20Result%20Structure%20%28by%20ref%29))
  (bind (_bblockinghit _binitialoverlap _time _distance _location) (Collision|BreakHitResult Hit))
  (if (ScalableFloat|GetValueasBool (Variables|Default|GetInput))
    (Building|SpawnBuildingActor (Variables|Default|GetBGAToSpawn) (Math|Transform|MakeTransform (- (.location (Math|Conversions|ToTransform(Vector) _location)) "0.000000,0.000000,125.000000") (.rotation (Math|Conversions|ToTransform(Vector) _location)) (.scale (Math|Conversions|ToTransform(Vector) _location))) (Variables|Actor|GetInstigator) (Variables|Actor|GetInstigator))))

(fn CheckObstructed
  (params (Target :type Fort%20Pawn%20Object%20Reference))
  (bind _outhit (Collision|LineTraceByChannel (Transformation|GetActorLocation) (Transformation|GetActorLocation Target) "TraceTypeQuery6"))
  (if (Collision|BreakHitResult _outhit)
    (bind _returnvalue (GameplayTags|HasAnyMatchingGameplayTags(Message) (.HitActor (Collision|BreakHitResult _outhit)) (Variables|Default|GetObstructorTags)))
    (return _returnvalue)
    (else
      (return false))))

(fn QuestUpdate
  ;[0 0 0] Check if elligible players were found
  (if (> (Variables|Default|GetCount) 0)
    ;[0 0 0] Send the verb update with the count set to the number of players hit
    (CustomVerb|BroadcastCustomVerbMessage (CustomVerb|MakeCustomVerbMessage (Actor|GetOwner) "(TagName=\"CustomVerb.Hit.DanceGrenade\")" (Variables|Default|GetCount)))))

(fn ConstructionScript
  (|CallParentFunction))
