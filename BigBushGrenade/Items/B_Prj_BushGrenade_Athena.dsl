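; Thrown bush grenade projectile, a BGA spawner child that spawns a bush building actor on a valid hit.
; On authority, valid hits spawn the bush, optionally explode with a quest event, and attach it to the hit actor, using scalable-float toggles and scales for vehicles.
; Invalid hits can revert it to a pickup, immediately or after repeated failures.
; Water entry splashes, disables collision and hides it; capsule hits clamp bounce speed.
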(event Custom|InitiateSpawnBGAChildEvent
  (params (HitResult :type Hit%20Result%20Structure) (HitOtherActor :type Actor%20Object%20Reference) (HitOtherComp :type Primitive%20Component%20Object%20Reference) (HitLocation :type Vector))
  (bind (_bblockinghit _binitialoverlap _time _distance _location _impactpoint _normal _impactnormal _physmat _hitactor _hitcomponent) (Collision|BreakHitResult HitResult))
  (bind (_asfort_aipawn _bsuccess) (Utilities|Casting|CastToFortAIPawn HitOtherActor))
  (bind _rootcomponent (Variables|/Script/Engine.Actor|GetRootComponent HitOtherActor))
  ;[0 0.009347 1] InitiateSpawnBGA
  ; Setup (stop simulating, do once) and then Spawn BGA on valid hit
  (Game|Components|ProjectileMovement|StopSimulating (Variables|Movement|GetProjectileMovementComponent) HitResult)
  (CallFunction|AdjustSpawnScaleforVehicles HitOtherActor)
  (bind _returnvalue (CallFunction|SpawnBGA _normal _hitactor _hitcomponent (Math|Vector|SelectVector HitLocation (CallFunction|GetClosestPoint HitOtherComp (Transformation|GetActorLocation)) _bsuccess) false HitResult))
  (bind _asbuilding_actor (Utilities|Casting|CastToBuildingActor _returnvalue))
  (Class|BuildingActor|SetCanbeSavedtoCreativeVolume _asbuilding_actor false) ; Thrown items shouldn't get saved by default to Creative maps
  (if (Variables|Default|GetDoExplosionOnBGASpawn)
    ;[1 1 1] Send Gameplay Event for Quest usage
    (Ability|SendGameplayEventtoActor (Actor|GetOwner) "(TagName=\"athena.event.gameplay.ThrowBush\")" (Utilities|Struct|MakeGameplayEventData "(TagName=\"athena.event.gameplay.ThrowBush\")" (Actor|GetOwner) 0 _returnvalue))
    (Delivery|DoExplosion)
    ;;[0.008226 1 0.1079] Allow spawning on vehicles
    (if (ScalableFloat|GetValueasBool (Variables|Default|GetAttachToVehiclesHF))
      (bind _returnvalue_2 (GameplayTags|HasMatchingGameplayTag(Message) HitOtherActor (Variables|Default|GetValet%20Vehicle%20Tag%20to%20Check)))
      (if _returnvalue_2
        (bind _returnvalue_3 (Class|FortVehicleInterface|GetSkeletalMeshComponent(Message) HitOtherActor))
        (if (Transformation|DoesSocketExist _returnvalue_3 (Variables|Default|GetDriverChoiceSocket))
          (Transformation|AttachActorToComponent _returnvalue _rootcomponent (Variables|Default|GetDriverChoiceSocket) "KeepWorld" "KeepWorld" "KeepWorld" true)
          (else
            _))
        (else
          ;;;[1 1 1] If we hit a cow catcher
          (bind _returnvalue_4 (GameplayTags|HasMatchingGameplayTag(Message) HitOtherActor (Variables|Default|GetCow%20Catcher%20Tag%20to%20Check)))
          (if _returnvalue_4
            _)))
      (Transformation|AttachActorToComponent _returnvalue _rootcomponent "None" "KeepWorld" "KeepWorld" "KeepWorld" true))))

(event EventTick
  (params (DeltaSeconds :type Float%20%28single-precision%29)))

(event OnComponentHit(CapsuleComponent)
  (params (HitComponent :type Primitive%20Component%20Object%20Reference) (OtherActor :type Actor%20Object%20Reference) (OtherComp :type Primitive%20Component%20Object%20Reference) (NormalImpulse :type Vector :default "0, 0, 0") (Hit :type Hit%20Result%20Structure%20%28by%20ref%29))
  (Projectile|SetVelocity (Variables|Movement|GetProjectileMovementComponent) (Math|Vector|ClampVectorSize (Variables|/Script/Engine.MovementComponent|GetVelocity (Variables|Movement|GetProjectileMovementComponent)) 0.0 1.5e+03)))

(event Collision|EventHit
  (params (MyComp :type Primitive%20Component%20Object%20Reference) (Other :type Actor%20Object%20Reference) (OtherComp :type Primitive%20Component%20Object%20Reference) (bSelfMoved :type Boolean) (HitLocation :type Vector :default "0, 0, 0") (HitNormal :type Vector :default "0, 0, 0") (NormalImpulse :type Vector :default "0, 0, 0") (Hit :type Hit%20Result%20Structure%20%28by%20ref%29))
  ;[1 1 1] Event Hit; Child version duped from parent B_Prj_Athena_BGASpawner. There's enough changes to base behavior that I don't want to dirty the parent for now.
  ; Check if the projectile has hit a valid target and determine the behavior of the hit
  (if (not (Variables|Default|GetHasHitWater))
    (bind _inairloop (Variables|Default|GetInAirLoop)
      (:then
        (Audio|Components|Audio|FadeOut _inairloop 0.25 0.0 "Linear" :!class /Script/Engine.AudioComponent))
      (:else
        _))
    (Networking|SwitchHasAuthority
      (:Authority
        (Default|IsHitValidforSpawn OtherComp Other Hit
          (:InvalidSpawn
            (CallFunction|ReverttoPickupAfterNFailedAttempts))
          (:ValidSpawn
            (CallFunction|InitiateSpawnBGAChildEvent Hit Other OtherComp HitLocation))
          (:InvalidSpawnAndRevertToPickup
            (CallFunction|RevertThrownObjecttoPickupChildEvent)))))))

(event EventOnStop
  (params (Hit :type Hit%20Result%20Structure%20%28by%20ref%29))
  (|CallParentFunction Hit))

(event EventOnExploded
  (params (HitActors :type Array%20of%20Actor%20Object%20References) (HitResults :type Array%20of%20Hit%20Result%20Structures))
  (CallFunction|DelayedExplode))

(event Custom|DelayedExplode
  (params (Hit :type Hit%20Result%20Structure))
  (bind (_bblockinghit _binitialoverlap _time _distance _location _impactpoint _normal _impactnormal _physmat _hitactor) (Collision|BreakHitResult Hit))
  (if (not (Variables|Default|GetHasExploded))
    (Variables|Default|SetHasExploded true)
    (GameplayCue|ExecuteGameplayCue%28Burst%29Local self (Variables|Default|GetImpactExplodeGC) (Ability|GameplayCue|MakeGameplayCueParameters :NormalizedMagnitude 0.0 :RawMagnitude 256.0 :Location _impactpoint :Normal _normal :EffectCauser _hitactor :PhysicalMaterial _physmat :GameplayEffectLevel 1 :AbilityLevel 1 :bReplicateLocationWhenUsingMinimalRepProxy false))
    (Utilities|HidePrjMeshandDelayDeathtoAllowFXToPlay)))

(event OnEnterWaterEvent(WaterInteractionComponent)
  (params (WaterBody :type Fort%20Water%20Body%20Actor%20Object%20Reference) (WaterInteractionComponent :type Fort%20Water%20Interaction%20Component%20Object%20Reference) (bIsFirstBody :type Boolean))
  ;[1 1 1] Duping from parent B_Prj_Athena_BGASpawner. We let this 'live' a little past normal to allow FX to play / fade out
  (Variables|Default|SetHasHitWater true)
  (if bIsFirstBody
    (CallFunction|PlaySplashVFX)
    (Actor|SetLifeSpan (Variables|Spawning|GetLifespanAfterSpawnBGA))
    (Collision|SetActorEnableCollision false)
    (Utilities|HidePrjMeshandDelayDeathtoAllowFXToPlay)))

(event Custom|PlaySplashVFX
  (Effects|Components|ParticleSystem|SpawnFXAtLocation "/Game/Effects/Niagara/Systems/Enviroment/Water/NS_WaterSplash_Medium.NS_WaterSplash_Medium" (Transformation|GetActorLocation)))

(event Custom|RevertThrownObjectToPickupChildEvent
  (Networking|SwitchHasAuthority
    (:Authority
      (bind _itemdeftospawnaspickupwheninvalid (Variables|Spawning|GetItemDefToSpawnAsPickupWhenInvalid)
        (:then
          (Inventory|SpawnPickupInWorld _itemdeftospawnaspickupwheninvalid 1 (Transformation|GetActorLocation) "0.000000,0.000000,1.000000" 0 true false false -1 "Tossed" "TossedByPlayer" 0 false))
        (:else
          ;[1 0 0] Warning: Item expected to revert to a pickup, but the Item Definition is invalid
          (Fortnite|FortShippingLog self (Utilities|String|Append (Utilities|GetDisplayName self) ": attempt to spawn thrown projectile as pickup failed (invalid ItemDef) - destroying projectile") true)))
      (Actor|SetLifeSpan (Variables|Spawning|GetLifespanAfterSpawnBGA))
      (Collision|SetActorEnableCollision false)
      (Utilities|HidePrjMeshandDelayDeathtoAllowFXToPlay))))

(fn AdjustSpawnScaleForVehicles
  (params (Hit%20Actor :type Actor%20Object%20Reference))
  (bind _returnvalue (GameplayTags|HasMatchingGameplayTag(Message) Hit%20Actor (Variables|Default|GetValet%20Vehicle%20Tag%20to%20Check)))
  (if _returnvalue
    (bind _returnvalue_2 (GameplayTags|HasMatchingGameplayTag(Message) Hit%20Actor (Variables|Default|GetBike%20Tag%20to%20Check)))
    (if _returnvalue_2
      (Variables|Spawning|TransformData|SetSpawnScale (Math|Conversions|ToVector(Float) (ScalableFloat|GetValueatLevel (Variables|Default|GetAttachToBikeScaleHF))))
      (else
        (Variables|Spawning|TransformData|SetSpawnScale (Math|Conversions|ToVector(Float) (ScalableFloat|GetValueatLevel (Variables|Default|GetAttachToVehicleScaleHF))))))
    (return)))

(fn RevertToPickupAfterNFailedAttempts
  (if (ScalableFloat|GetValueasBool (Variables|Default|GetShouldRevertToPickupOnFailedAttemptsHF))
    (bind _output_get (Variables|Default|SetFailedSpawnAttempts (+ (Variables|Default|GetFailedSpawnAttempts) 1)))
    (if (> _output_get (ScalableFloat|GetValueasInteger (Variables|Default|GetNumFailedAttemptsCheckHF)))
      (CallFunction|RevertThrownObjecttoPickupChildEvent))))

(fn ShouldBounceOffActor
  (params (Hit%20Actor :type Actor%20Object%20Reference) (Hit%20Component :type Primitive%20Component%20Object%20Reference))
  (bind (_asbuilding_prop _bsuccess) (Utilities|Casting|CastToBuildingProp Hit%20Actor))
  (bind (_asfort_athena_supply_drop _bsuccess_2) (Utilities|Casting|CastToFortAthenaSupplyDrop Hit%20Actor))
  (bind (_asfort_projectile_base _bsuccess_3) (Utilities|Casting|CastToFortProjectileBase Hit%20Actor))
  (bind (_asathena_supply_drop_balloon _bsuccess_4) (Utilities|Casting|CastToAthenaSupplyDropBalloon Hit%20Actor))
  (bind (_asfort_pawn _bsuccess_5) (Utilities|Casting|CastToFortPawn Hit%20Actor))
  (Utilities|IsValid Hit%20Actor
    (:Is%20Valid
      (Utilities|IsValid Hit%20Component
        (:Is%20Valid
          (return (Math|Boolean|ORBoolean (Utilities|Array|ContainsItem (Variables|/Script/Engine.ActorComponent|GetComponentTags Hit%20Component) (LocalVariables|GetBounce%20Tag)) (Utilities|Array|ContainsItem (Variables|/Script/Engine.Actor|GetTags Hit%20Actor) (LocalVariables|GetBounce%20Tag)) _bsuccess _bsuccess_5 _bsuccess_2 _bsuccess_4 _bsuccess_3)))
        (:Is%20Not%20Valid
          _)))
    (:Is%20Not%20Valid
      _))
  (return false))

(fn ConstructionScript
  (|CallParentFunction)
  (Components|Movement|SetUpdatedComponent (Variables|Components|GetRotatingMovement) (Variables|Components|GetLeafGrenade)))
