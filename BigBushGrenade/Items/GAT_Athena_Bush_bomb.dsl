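; Gameplay ability for the bush bomb item that strips the bush disguise from the player, apparently after damage.
; On activation it calls the parent, caches player 0's pawn, removes gameplay effects granting the Athena.Item.Bush tag and forces a net update.
; RemoveBush then destroys attached Athena_Player_BushBomb actors and ends the ability.
; The commented charge-based item removal has only a controller cast, with no removal logic.
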(event Ability|EventActivateAbility
  (bind _playerpawn (Variables|Default|GetPlayerPawn))
  (|CallParentFunction)
  (bind _asfort_player_pawn (Utilities|Casting|CastToFortPlayerPawn (Game|GetPlayerPawn 0)))
  (Variables|Default|SetPlayerPawn _asfort_player_pawn)
  ;[1 1 1] Remove Item from player if they don't have any charges left
  (bind _asfort_player_controller (Utilities|Casting|CastToFortPlayerController (Pawn|GetController (Variables|Default|GetPlayerPawn)))
    (:then
      _)
    (:CastFailed
      _))
  (Ability|RemoveGameplayEffectFromOwnerWithGrantedTags "(GameplayTags=((TagName=\"Athena.Item.Bush\")))")
  (Networking|ForceNetUpdate _playerpawn)
  ;[1 1 1] Remove the Bush in case they took damage
  (CallFunction|RemoveBush _playerpawn))

(fn RemoveBush
  (params (PlayerPawn :type Fort%20Player%20Pawn%20Object%20Reference))
  (for _array_element (Actor|GetAttachedActors PlayerPawn true false)
    (bind _asathena_player_bush_bomb (Utilities|Casting|CastToAthena_Player_BushBomb _array_element))
    (Utilities|IsValid _asathena_player_bush_bomb
      (:Is%20Valid
        (Actor|DestroyActor _asathena_player_bush_bomb))))
  (Ability|EndAbility))
